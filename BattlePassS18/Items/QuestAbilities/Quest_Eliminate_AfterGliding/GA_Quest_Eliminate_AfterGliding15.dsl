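; Gameplay ability for the Battle Pass S18 quest on eliminations after gliding, applying the quest's effect to the owner once gliding ends.
; On event activation it calls the parent, then waits once for the parachuting tag to be removed.
; When the tag is removed, it applies the Eliminate After Gliding gameplay effect and ends itself.
; Debug prints, gated by a debug flag, trace activation and effect application.
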(event Ability|EventActivateAbilityFromEvent
  (params (EventData :type Gameplay%20Event%20Data%20Structure%20%28by%20ref%29))
  (|CallParentFunction EventData)
  (Development|PrintString "@@@ ActivateAbilityFromEvent [GA_Quest_Eliminate_AfterGliding15]: CALLED!" false (Variables|Debug|GetbDebugAbility))
  (Ability|Tasks|WaitGameplayTagRemove (Variables|Default|GetT_Parachuting) 0 true
    (:Removed
      (Development|PrintString "@@@@@@ ActivateAbilityFromEvent [GA_Quest_Eliminate_AfterGliding15]: Applying Eliminate After Gliding effect!" false (Variables|Debug|GetbDebugAbility))
      (Ability|ApplyGameplayEffectToOwner (Variables|Default|GetGE_Quest_Eliminate_AfterGliding))
      (Ability|EndAbility))) ; Wait for the player to stop parachuting
)
